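# Character body material for the female Genius player skin, giving it an anime toon look on skinned meshes with the Cloth shading model.
# It chains material attributes from the shared character PreFX function through the shared anime shading function.
# The result then passes through a Genius-specific eye animation function and a shared toon flipbook mouth function before entering the character PostFX function.
# All logic lives in those material functions.
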Material /Game/Characters/Player/Female/Medium/Bodies/F_MED_Genius/Materials/M_F_Genius_AnimeShading
graph {
 properties {
  bAutomaticallySetUsageInEditor=false
  bUseMaterialAttributes=true
  bUsedWithClothing=true
  bUsedWithMeshDeformer=true
  bUsedWithMorphTargets=true
  bUsedWithSkeletalMesh=true
  pixelDepthOffsetMode=PDOM_Legacy
  shadingModel=MSM_Cloth
  usedShadingModels=Cloth
 }
 nodes [
  n1 = MaterialFunctionCall() {materialFunction=&BaseFN_CharacterFunction_PreFX}
  n2 = MaterialFunctionCall!(
   Material=@n1.Result
   LitDiffuseOverride
   SSC_AlphaOverride
   Tint Mask
   DF InkLine Visibility Mask
   DF Inkline Injection
   DF Inkline Color Override
   DF Inkline Color Override Mask
   CustomUVs
  ) {materialFunction=/Game/Packages/Fortress_SharedMaterials/Anime_Master_Material/MF_AnimeShading}
  MaterialFunctionCall(
   AO=@n1.AO
   Distance Blend=@n1.AO
   Fresnel=@n1.AO
   Gmap=@n1.AO
   Pre Skinned Local Position=@n1.AO
   Material=MaterialFunctionCall!(In=MaterialFunctionCall!(MA=@n2.Result) {materialFunction=&/MF_Genius_EyeAnimation}.Result) {materialFunction=/Game/Packages/Fortress_SharedMaterials/Anime_Master_Material/MF_ToonFlipbookMouth}.Result
  ) {materialFunction=&BaseFN_CharacterFunction_PostFX}
 ]
}
